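# HealingGrenade_Explode_Cue is a sound cue for the healing grenade's explosion.
# A random node picks one of four explosion wave variants with equal weight and feeds a modulator that holds pitch unchanged.
# It uses smoke-grenade attenuation and the large-explosives sound class at reduced volume.
# Its concurrency override prevents new instances from starting once the limit is reached.
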SoundCue /FNGameplayCues/Sounds/Fort_Weapons/Launchers/Healing/HealingGrenade_Explode_Cue
root=Modulator_0
nodes [
 {id=Modulator_0, children [Random_0], properties {pitchMax=1, pitchMin=1}}
 {id=Random_0, children [WavePlayer_1,WavePlayer_2,WavePlayer_3,WavePlayer_4], properties {weights [1,1,1,1]}}
 {id=WavePlayer_1, properties {soundWaveAssetPtr=&/HealingGrenade_Explode_04}}
 {id=WavePlayer_2, properties {soundWaveAssetPtr=&/HealingGrenade_Explode_01}}
 {id=WavePlayer_3, properties {soundWaveAssetPtr=&/HealingGrenade_Explode_02}}
 {id=WavePlayer_4, properties {soundWaveAssetPtr=&/HealingGrenade_Explode_03}}
]
properties {
 attenuationSettings=/Game/Sounds/Attenuation/Fort_Grenade_Smoke_Att
 bOverrideConcurrency=true
 maxDistance=6800
 soundClassObject=/Game/Sounds/FortSoundClassesAndModes/Fort_Explosives_Large_PSM
 volumeMultiplier=0.6369
 attenuationOverrides {falloffDistance=2400, attenuationShapeExtents {x=600}}
 concurrencyOverrides {resolutionRule=PreventNew}
}
