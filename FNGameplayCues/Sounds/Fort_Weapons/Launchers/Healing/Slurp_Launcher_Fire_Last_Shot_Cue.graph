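# Sound cue for the last-shot fire sound of the Slurp healing launcher weapon.
# A single last-shot wave plays through a slight random pitch modulator, then a close last-shot projectile attenuation.
# The root node overrides the sound class to third-person ranged weapon fire.
# Local and global gunfire concurrency sets limit simultaneous playback.
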SoundCue /FNGameplayCues/Sounds/Fort_Weapons/Launchers/Healing/Slurp_Launcher_Fire_Last_Shot_Cue
root=SoundClass_32
nodes [
 {
  id=Attenuation_49
  children [Modulator_2]
  properties {attenuationSettings=/FNGameplayCues/Sounds/Attenuation/Fort_Weapon_Projectile_Close_Last_Shot_Athena_Att}
 }
 {id=Modulator_2, children [WavePlayer_2], properties {pitchMax=1.15, pitchMin=0.9, volumeMax=1, volumeMin=1}}
 {
  id=SoundClass_32
  children [Attenuation_49]
  properties {soundClassOverride=&Fort_WeaponRanged_Fire_3P_PSM}
 }
 {id=WavePlayer_2, properties {soundWaveAssetPtr=&/WPN_Launcher_Slurp_Last_Shot_01}}
]
properties {
 maxDistance=1800.0001
 volumeMultiplier=1.3
 attenuationOverrides {distanceAlgorithm=NaturalSound, falloffDistance=5000, attenuationShapeExtents {x=1000}}
 concurrencySet [&Weapon_Gunfire_Local_Conc, &Weapon_Gunfire_Global_Conc]
}
