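# Sound cue for the Slurp Launcher projectile explosion, mixing close, mid and far layers.
# Each layer randomly picks one of four waves and has its own distance attenuation and sound class override, with mid and far sharing the distant class.
# The close layer fades in through an enveloper, and the far layer gets a short random delay; both vary pitch.
# A concurrency override blocks new instances at the limit.
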SoundCue /FNGameplayCues/Sounds/Fort_Weapons/Launchers/Healing/Slurp_Launcher_Proj_Explode_Cue
root=Mixer_5
nodes [
 {id=Attenuation_4, children [Modulator_13], properties {attenuationSettings=/Game/Sounds/Attenuation/Fort_Explosion_Medium_Athena_Att}}
 {
  id=Attenuation_6
  children [Modulator_6]
  properties {
   attenuationSettings=/Game/Sounds/Attenuation/Fort_Explosion_Close_Athena_Att
   attenuationOverrides {falloffDistance=900, attenuationShapeExtents {x=100}}
  }
 }
 {
  id=Attenuation_8
  children [Modulator_8]
  properties {
   attenuationSettings=/Game/Sounds/Attenuation/Fort_Explosion_Distant_Athena_Att
   attenuationOverrides {falloffDistance=3000, attenuationShapeExtents {x=1000}}
  }
 }
 {id=Delay_0, children [Attenuation_8], properties {delayMax=0.1}}
 {
  id=Enveloper_0
  children [Random_4]
  properties {pitchCurve {editorCurveData {keys [{time=0}]}}, volumeCurve {editorCurveData {keys [<at 0 0.0028>,<at 0.2038 1.0018>]}}}
 }
 {id=Mixer_5, children [SoundClass_3,SoundClass_2,SoundClass_0], properties {inputVolume [0.4,0.4,0.6]}}
 {id=Modulator_13, children [Random_5], properties {pitchMax=1, pitchMin=1, volumeMax=1, volumeMin=1}}
 {id=Modulator_6, children [Enveloper_0], properties {pitchMax=1.1, pitchMin=0.9, volumeMax=1, volumeMin=1}}
 {id=Modulator_8, children [Random_6], properties {pitchMax=1.3, pitchMin=0.75, volumeMax=1, volumeMin=1}}
 {id=Random_4, children [WavePlayer_8, WavePlayer_9, WavePlayer_20, WavePlayer_21], properties {weights [1,1,1,1]}}
 {id=Random_5, children [WavePlayer_22, WavePlayer_23, WavePlayer_24, WavePlayer_25], properties {weights [1,1,1,1]}}
 {id=Random_6, children [WavePlayer_26, WavePlayer_27, WavePlayer_28, WavePlayer_29], properties {weights [1,1,1,1]}}
 {id=SoundClass_0, children [Delay_0], properties {soundClassOverride=&Fort_WeaponRanged_Fire_Distant}}
 {
  id=SoundClass_2
  children [Attenuation_4]
  properties {soundClassOverride=&Fort_WeaponRanged_Fire_Distant}
 }
 {
  id=SoundClass_3
  children [Attenuation_6]
  properties {soundClassOverride=&Fort_WeaponRanged_Fire_3P_PSM}
 }
 {
  id=WavePlayer_20
  properties {soundWaveAssetPtr=&/WPN_Launcher_Slurp_Proj_Explo_Close_03}
 }
 {
  id=WavePlayer_21
  properties {soundWaveAssetPtr=&/WPN_Launcher_Slurp_Proj_Explo_Close_04}
 }
 {id=WavePlayer_22, properties {soundWaveAssetPtr=&/WPN_Launcher_Slurp_Proj_Explo_Mid_01}}
 {id=WavePlayer_23, properties {soundWaveAssetPtr=&/WPN_Launcher_Slurp_Proj_Explo_Mid_02}}
 {id=WavePlayer_24, properties {soundWaveAssetPtr=&/WPN_Launcher_Slurp_Proj_Explo_Mid_03}}
 {id=WavePlayer_25, properties {soundWaveAssetPtr=&/WPN_Launcher_Slurp_Proj_Explo_Mid_04}}
 {id=WavePlayer_26, properties {soundWaveAssetPtr=&/WPN_Launcher_Slurp_Proj_Explo_Far_01}}
 {id=WavePlayer_27, properties {soundWaveAssetPtr=&/WPN_Launcher_Slurp_Proj_Explo_Far_02}}
 {id=WavePlayer_28, properties {soundWaveAssetPtr=&/WPN_Launcher_Slurp_Proj_Explo_Far_03}}
 {id=WavePlayer_29, properties {soundWaveAssetPtr=&/WPN_Launcher_Slurp_Proj_Explo_Far_04}}
 {id=WavePlayer_8, properties {soundWaveAssetPtr=&/WPN_Launcher_Slurp_Proj_Explo_Close_01}}
 {id=WavePlayer_9, properties {soundWaveAssetPtr=&/WPN_Launcher_Slurp_Proj_Explo_Close_02}}
]
properties {
 bOverrideConcurrency=true
 maxDistance=27500
 volumeMultiplier=1
 attenuationOverrides {falloffDistance=2400, attenuationShapeExtents {x=600}}
 concurrencyOverrides {resolutionRule=PreventNew}
}
